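; Static Living World point provider that serves as the spawner for the ToyFox boss.
; Its custom EnableProvider event checks for a dedicated server, then enables the point provider through the deprecated interface call and asks the Living World Manager to generate events.
; The construction script is empty, and the graph itself contains no direct spawn logic.
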(event Custom|EnableProvider
  (if (Networking|IsDedicatedServer)
    (Class|FortAthenaLivingWorldPointProviderInterface|EnablePointProvider_DEPRECATED)
    (Class|FortAthenaLivingWorldManager|RequestEventGeneration (Class|FortAthenaLivingWorldManager|GetLivingWorldManager))))

(fn ConstructionScript)
